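# One-shot mono MetaSound for the destruction of a wooden ModBox in VehicleMods.
# On play, a random-get node picks one of four destroy wave variants and feeds it to a wave player, whose completion finishes the source.
# The source uses a dedicated attenuation asset and the Fortnite explosives sound class, with overridden concurrency to limit simultaneous and rapidly retriggered plays.
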MetaSoundSource /VehicleMods/Gameplay/Actors/ModBox/Spawners/Audio/MSS_Modbox_Destroyed_Wood
inputs [
 Destroy Sounds {type=WaveAsset:Array}
 UE.Source.OnPlay {type=Trigger}
]
interfaces [UE.OutputFormat.Mono:1.0, UE.Source.OneShot:1.0, UE.Source:1.1]
outputs [
 "UE.OutputFormat.Mono.Audio:0" {type=Audio}
 UE.Source.OneShot.OnFinished {type=Trigger}
]
pages [Default {
  input_defaults {
   UE.Source.OnPlay=null
   Destroy Sounds [
    /VehicleMods/Gameplay/Actors/ModBox/Spawners/Audio/wav/ModBox_Destroy_01
    /VehicleMods/Gameplay/Actors/ModBox/Spawners/Audio/wav/ModBox_Destroy_02
    /VehicleMods/Gameplay/Actors/ModBox/Spawners/Audio/wav/ModBox_Destroy_03
    /VehicleMods/Gameplay/Actors/ModBox/Spawners/Audio/wav/ModBox_Destroy_04
   ]
  }
  nodes [
   ArrayRandomGet_0 = Array.Random Get:WaveAsset:Array(In Array=^Destroy Sounds, Next=^UE.Source.OnPlay) {also {version=~}}
   WavePlayer_0 = UE.Wave Player:Mono(Play=@ArrayRandomGet_0.On Next, Wave Asset=@ArrayRandomGet_0.Value)
  ]
  links [{dst="@UE.Source.OneShot.OnFinished", src=WavePlayer_0.On Finished}, {dst="@UE.OutputFormat.Mono.Audio:0", src=WavePlayer_0.Out Mono}]
 }]
properties {
 assetImportData=&self:AssetImportData
 attenuationSettings=&/Att_ModBoxDestroy
 bOverrideConcurrency=true
 maxDistance=8500
 sampleRate=48000
 soundClassObject=/Game/Sounds/FortSoundClassesAndModes/Fort_Explosives
 volume=0.425
 concurrencyOverrides {maxCount=3, retriggerTime=0.3}
}
