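# Mono MetaSound source for the vehicle ModBox hologram sound, named as a loop and using ambient-loop attenuation with limited concurrency.
# On play, a random pick from two hologram waves, held in an input misleadingly named Destroy Sounds, feeds a wave player.
# An ADSR envelope with full sustain multiplies the audio to fade it in.
# The wave player's finish trigger drives the OneShot OnFinished output.
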MetaSoundSource /VehicleMods/Gameplay/Actors/ModBox/Spawners/Audio/MSS_Modbox_Holo_lp
inputs [
 Destroy Sounds {type=WaveAsset:Array}
 UE.Source.OnPlay {type=Trigger}
]
interfaces [UE.OutputFormat.Mono:1.0, UE.Source.OneShot:1.0, UE.Source:1.1]
outputs [
 "UE.OutputFormat.Mono.Audio:0" {type=Audio}
 UE.Source.OneShot.OnFinished {type=Trigger}
]
pages [Default {
  input_defaults {
   UE.Source.OnPlay=null
   Destroy Sounds [
    /VehicleMods/Gameplay/Actors/ModBox/Spawners/Audio/wav/ModBox_Holo_lp_01
    /VehicleMods/Gameplay/Actors/ModBox/Spawners/Audio/wav/ModBox_Holo_lp_02
   ]
  }
  nodes [
   ADSREnvelopeADSREnvelope_0 = ADSR Envelope.ADSR Envelope:Audio(Trigger Attack=^UE.Source.OnPlay) {also {defaults {Attack Time=1, Sustain Level=1}}}
   ArrayRandomGet_0 = Array.Random Get:WaveAsset:Array(In Array=^Destroy Sounds, Next=^UE.Source.OnPlay)
   WavePlayer_0 = UE.Wave Player:Mono(Play=@ArrayRandomGet_0.On Next, Wave Asset=@ArrayRandomGet_0.Value) {also {defaults {Loop="<MetaSoundBuilderResult.SUCCEEDED: 0>"}}}
   Multiply_0 = UE.Multiply:Audio(AdditionalOperands=@ADSREnvelopeADSREnvelope_0.Out Envelope, PrimaryOperand=@WavePlayer_0.Out Mono)
  ]
  links [{dst="@UE.OutputFormat.Mono.Audio:0", src=Multiply_0.Out}, {dst="@UE.Source.OneShot.OnFinished", src=WavePlayer_0.On Finished}]
 }]
properties {
 assetImportData=&self:AssetImportData
 attenuationSettings=/Game/Sounds/Attenuation/Fort_Item_Ambient_Loop_Mono_Default_Att
 bOverrideConcurrency=true
 maxDistance=1250
 sampleRate=48000
 soundClassObject=&SFX
 virtualizationMode=PlayWhenSilent
 volume=2
 concurrencyOverrides {maxCount=2, retriggerTime=0.1, volumeScale=0.8, volumeScaleMode=Distance}
}
