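# One-shot mono MetaSound Source for the respawn sound of mod boxes in VehicleMods.
# On play, a random-get node picks from the Respawn Sound array, which defaults to a single wave, and triggers a mono Wave Player.
# The player feeds the audio output and signals completion when finished.
# It uses the SFX class, mod box ambience attenuation and overridden per-owner concurrency with a small voice cap and retrigger time.
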MetaSoundSource /VehicleMods/Gameplay/Actors/ModBox/Spawners/Audio/MSS_Modbox_Respawn
inputs [
 Respawn Sound {type=WaveAsset:Array}
 UE.Source.OnPlay {type=Trigger}
]
interfaces [UE.OutputFormat.Mono:1.0, UE.Source.OneShot:1.0, UE.Source:1.1]
outputs [
 "UE.OutputFormat.Mono.Audio:0" {type=Audio}
 UE.Source.OneShot.OnFinished {type=Trigger}
]
pages [Default {
  input_defaults {UE.Source.OnPlay=null, Respawn Sound [/VehicleMods/Gameplay/Actors/ModBox/Spawners/Audio/wav/Modbox_Respawn_01]}
  nodes [
   ArrayRandomGet_0 = Array.Random Get:WaveAsset:Array(In Array=^Respawn Sound, Next=^UE.Source.OnPlay) {also {version=~}}
   WavePlayer_0 = UE.Wave Player:Mono(Play=@ArrayRandomGet_0.On Next, Wave Asset=@ArrayRandomGet_0.Value)
  ]
  links [{dst="@UE.Source.OneShot.OnFinished", src=WavePlayer_0.On Finished}, {dst="@UE.OutputFormat.Mono.Audio:0", src=WavePlayer_0.Out Mono}]
 }]
properties {
 assetImportData=&self:AssetImportData
 attenuationSettings=&/Att_ModBoxAmb
 bOverrideConcurrency=true
 maxDistance=1500
 sampleRate=48000
 soundClassObject=&SFX
 concurrencyOverrides {bLimitToOwner=true, maxCount=2, retriggerTime=0.4}
}
